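# Looping quest sound cue for a server ambience, going by its name for the Sifting Through The Archives quest.
# A Mixer blends an agency computer server loop, weighted louder, with a quieter reused supply drop target loop, then feeds a default Modulator.
# An endlessly looping Enveloper gently dips and restores the server layer's volume over a twenty-second cycle.
# It primes on load and borrows the trash-flies attenuation asset from another quest.
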SoundCue /Game/Athena/Sounds/Quests/SiftingThroughTheArchives/S20_SiftingThroughTheArchives_Server_Loop_Cue
root=Modulator_1
nodes [
 {
  id=Enveloper_2
  children [WavePlayer_7]
  properties {
   bLoop=true
   bLoopIndefinitely=true
   loopEnd=20
   pitchCurve {editorCurveData {keys [{time=0}]}}
   volumeCurve {editorCurveData {keys [
      {arriveTangent=-0.01, interpMode=RCIM_Cubic, time=0, value=1}
      <at 10 0.8 RCIM_Cubic>
      {interpMode=RCIM_Cubic, leaveTangent=0.01, time=20, value=1}
     ]}}
  }
 }
 {id=Mixer_0, children [Enveloper_2,WavePlayer_0], properties {inputVolume [1.5,0.5]}}
 {id=Modulator_1, children [Mixer_0]}
 {id=WavePlayer_0, properties {bLooping=true, soundWaveAssetPtr=/Game/Sounds/Foley_Loot/Containers/SupplyDrops/sfx_supplydrop_target_loop}}
 {
  id=WavePlayer_7
  properties {bLooping=true, soundWaveAssetPtr=&/Spot_Agency_Computer_Server_Loop}
 }
]
properties {
 attenuationSettings=/Game/Athena/Sounds/Quests/CleaningUpTheMess/Spot_Flies_Around_Trash_Att
 bPrimeOnLoad=true
 maxDistance=2010
 soundClassObject=&Fort_Quests
 volumeMultiplier=1
 attenuationOverrides {falloffDistance=700, attenuationShapeExtents {x=100}}
}
