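# Sound cue that plays the server success stinger for the Season 20 "Sifting Through The Archives" quest.
# A Random node picks between two equally weighted Creative radio success stingers and feeds a root Mixer, whose second input is empty.
# It uses the UI sound class with a very large max distance, and its concurrency override allows one instance at a time, stopping the oldest.
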SoundCue /Game/Athena/Sounds/Quests/SiftingThroughTheArchives/S20_SiftingThroughTheArchives_Server_Success_Cue
root=Mixer_0
nodes [
 {id=Mixer_0, children [Random_1,null], properties {inputVolume [1,1]}}
 {id=Random_1, children [WavePlayer_4,WavePlayer_5], properties {weights [1,1]}}
 {id=WavePlayer_4, properties {soundWaveAssetPtr=/Game/Sounds/Creative/Gadgets/Radio/Stingers/Stinger_Success_01}}
 {id=WavePlayer_5, properties {soundWaveAssetPtr=/Game/Sounds/Creative/Gadgets/Radio/Stingers/Stinger_Success_02}}
]
properties {
 bOverrideConcurrency=true
 maxDistance=2097152
 soundClassObject=&Fort_UI
 volumeMultiplier=1
 concurrencyOverrides {maxCount=1, resolutionRule=StopOldest}
}
